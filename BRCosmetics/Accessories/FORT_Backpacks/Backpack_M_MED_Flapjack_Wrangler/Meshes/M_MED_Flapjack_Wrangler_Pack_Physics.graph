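# Physics asset for the Flapjack Wrangler backpack cosmetic, giving capsule bodies and joint constraints to its chains, flaps, folds and syrup bones.
# A base rotate body hangs off the root with a very tight twist limit and carries a long chain, a short chain, the upper folds and a wobble mass.
# The rotation-locked wobble mass anchors three syrup bodies and three flaps, with each flap limited to swing mainly on one axis.
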PhysicsAsset /BRCosmetics/Accessories/FORT_Backpacks/Backpack_M_MED_Flapjack_Wrangler/Meshes/M_MED_Flapjack_Wrangler_Pack_Physics
bodies {
 dyn_back_flap_jnt [{shapeType=Capsule, length=2, radius=2, center=<0.9077 0.2884 -0.0641>, rotation=<rot 0 20.0861 -76.3997>}]
 dyn_base_rotate_jnt [{shapeType=Capsule, length=9, radius=5, center=<0 -4.9499 -6.9886>}]
 dyn_front_flap_jnt [{shapeType=Capsule, length=2, radius=2, center=<1.4247 2.4784 0.1467>, rotation=<rot 0 174.9386 -81.8702>}]
 dyn_long_chain_1_jnt [{shapeType=Capsule, length=1.55, radius=1.1, center=<2.5308 1.9067 0>, rotation=<rot 90 165.9637 125.2991>}]
 dyn_long_chain_2_jnt [{shapeType=Capsule, length=1.55, radius=1.1, center=<2.7818 1.9067 0>, rotation=<rot 90 165.9637 125.2991>}]
 dyn_long_chain_3_jnt [{shapeType=Capsule, length=1.55, radius=1.1, center=<2.6826 1.9067 0>, rotation=<rot 90 -14.0363 -54.7009>}]
 dyn_long_chain_4_jnt [{shapeType=Capsule, length=2.55, radius=1.1, center=<3.5021 2.8166 0>, rotation=<rot 90 0 -40.6646>}]
 dyn_long_chain_5_jnt [{shapeType=Capsule, length=2.55, radius=2.1, center=<5.6373 4.5616 0>, rotation=<rot 90 180 139.3354>}]
 dyn_short_chain_1_jnt [{shapeType=Capsule, length=1.55, radius=1.1, center=<2.4075 1.9067 0.0001>, rotation=<rot 90 165.9637 125.2992>}]
 dyn_short_chain_2_jnt [{shapeType=Capsule, length=1.55, radius=1.1, center=<2.5083 1.9067 0.0001>, rotation=<rot 90 165.9637 125.2991>}]
 dyn_short_chain_3_jnt [{shapeType=Capsule, length=1.55, radius=1.1, center=<2.2262 1.9067 0>, rotation=<rot 90 -14.0363 -54.7009>}]
 dyn_short_chain_4_jnt [{shapeType=Capsule, length=2.55, radius=2.1, center=<3.5407 2.838 0>, rotation=<rot 90 180 139.3354>}]
 dyn_side_flap_jnt [{shapeType=Capsule, length=2, radius=2, center=<1.0959 0.1559 -0.0463>, rotation=<rot 0 -0.7316 -84.1281>}]
 dyn_syrup_jnt_1 [{shapeType=Capsule, length=1.5, radius=1.5, center=<2.4694 0 -0.2639>, rotation=<rot 83.901 0 0>}]
 dyn_syrup_jnt_2 [{shapeType=Capsule, length=1.5, radius=1.5, center=<2.4695 0 -0.2631>, rotation=<rot 83.9177 0 0>}]
 dyn_syrup_jnt_3 [{shapeType=Capsule, length=1.5, radius=1.5, center=<2.4695 0 -0.2632>, rotation=<rot 83.9153 0 0>}]
 dyn_upper_folds_jnt [{shapeType=Capsule, length=7.3386, radius=2.75, center=<1.3 0 0>}]
 dyn_wobble_mass_jnt [{shapeType=Capsule, length=5.5, radius=11, center=<2.8987 -1.4653 0>}]
 root [{shapeType=Capsule, length=9.5, radius=2.5, center=<0 2.6538 0>}]
}
constraints [
 {bone1Name=dyn_long_chain_1_jnt, bone2Name=dyn_base_rotate_jnt, swing1LimitDegrees=1, swing2LimitDegrees=10, twistLimitDegrees=10}
 {bone1Name=dyn_long_chain_2_jnt, bone2Name=dyn_long_chain_1_jnt, swing1LimitDegrees=15, swing2LimitDegrees=10, twistLimitDegrees=10}
 {bone1Name=dyn_long_chain_3_jnt, bone2Name=dyn_long_chain_2_jnt, swing1LimitDegrees=15, swing2LimitDegrees=10, twistLimitDegrees=10}
 {bone1Name=dyn_long_chain_4_jnt, bone2Name=dyn_long_chain_3_jnt, swing1LimitDegrees=15, swing2LimitDegrees=10, twistLimitDegrees=10}
 {bone1Name=dyn_long_chain_5_jnt, bone2Name=dyn_long_chain_4_jnt, swing1LimitDegrees=15, swing2LimitDegrees=10, twistLimitDegrees=10}
 {bone1Name=dyn_short_chain_1_jnt, bone2Name=dyn_base_rotate_jnt, swing1LimitDegrees=3, swing2LimitDegrees=10, twistLimitDegrees=10}
 {bone1Name=dyn_short_chain_2_jnt, bone2Name=dyn_short_chain_1_jnt, swing1LimitDegrees=15, swing2LimitDegrees=12, twistLimitDegrees=12}
 {bone1Name=dyn_short_chain_3_jnt, bone2Name=dyn_short_chain_2_jnt, swing1LimitDegrees=15, swing2LimitDegrees=12, twistLimitDegrees=12}
 {bone1Name=dyn_short_chain_4_jnt, bone2Name=dyn_short_chain_3_jnt, swing1LimitDegrees=15, swing2LimitDegrees=12, twistLimitDegrees=12}
 {
  bone1Name=dyn_wobble_mass_jnt
  bone2Name=dyn_base_rotate_jnt
  swing1LimitDegrees=10
  swing1Motion=Locked
  swing2LimitDegrees=10
  swing2Motion=Locked
  twistLimitDegrees=10
  twistMotion=Locked
 }
 {bone1Name=dyn_syrup_jnt_1, bone2Name=dyn_wobble_mass_jnt, swing1LimitDegrees=10, swing2LimitDegrees=5, twistLimitDegrees=2}
 {bone1Name=dyn_syrup_jnt_2, bone2Name=dyn_wobble_mass_jnt, swing1LimitDegrees=10, swing2LimitDegrees=5, twistLimitDegrees=2}
 {bone1Name=dyn_syrup_jnt_3, bone2Name=dyn_wobble_mass_jnt, swing1LimitDegrees=10, swing2LimitDegrees=5, twistLimitDegrees=2}
 {bone1Name=dyn_upper_folds_jnt, bone2Name=dyn_base_rotate_jnt, swing1LimitDegrees=5, swing2LimitDegrees=5, twistLimitDegrees=5}
 {bone1Name=dyn_side_flap_jnt, bone2Name=dyn_wobble_mass_jnt, swing1LimitDegrees=5, swing2LimitDegrees=35, twistLimitDegrees=1}
 {bone1Name=dyn_back_flap_jnt, bone2Name=dyn_wobble_mass_jnt, swing1LimitDegrees=3, swing2LimitDegrees=40, twistLimitDegrees=1}
 {bone1Name=dyn_front_flap_jnt, bone2Name=dyn_wobble_mass_jnt, swing1LimitDegrees=5, swing2LimitDegrees=30, twistLimitDegrees=1}
 {bone1Name=dyn_base_rotate_jnt, bone2Name=root, swing1LimitDegrees=0, swing2LimitDegrees=0, twistLimitDegrees=3}
]
