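# Animation Blueprint for the Flapjack Wrangler backpack cosmetic, adding secondary physics motion to the pack skeleton.
# A RigidBody node simulates in pack-bone space, with much weaker component motion scaling in the front end.
# An AnimDynamics node then bobs the wobble mass joint along Z, and fixed additive rolls offset the base and two chain joints.
# Both simulations run at partial alpha at LOD 0 and drop to zero at higher LODs.
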AnimBlueprint /BRCosmetics/Accessories/FORT_Backpacks/Backpack_M_MED_Flapjack_Wrangler/Meshes/M_MED_Flapjack_Wrangler_Pack_AnimBP
target_skeleton=&/M_MED_Flapjack_Wrangler_Pack_Skeleton
graphs [
 AnimGraph {nodes [
   n1 = SelectVector(A="0.02,0.02,0.02", B="0.4,0.4,0.4", bPickA=Get bIsInFrontEnd())
   n2 = SelectFloat(A=0, B=0.8, bPickA="integer >= integer"!(A:int=Get CurrentLOD(), B:int=1) -> (ReturnValue:bool) {is=title K2Node_CallFunction})
   n3 = RigidBody!(
    ComponentPose=Local To Component(LocalPose=Input Pose())
    Alpha=@n2
    ExternalForce
    ComponentLinearAccScale:Vector=@n1
    ComponentLinearVelScale:Vector=@n1
   ) -> (Pose) {
    bClampLinearTranslationLimitToRefPose=true
    bOverrideWorldGravity=true
    overridePhysicsAsset=&/M_MED_Flapjack_Wrangler_Pack_Physics
    simulationSpace=BaseBoneSpace
    alphaBoolBlend {blendOption=Linear}
    baseBoneRef {boneName=pack}
    componentAppliedLinearAccClamp=<3000 3000 3000>
    componentLinearAccScale=<0.4 0.4 0.4>
    componentLinearVelScale=<0.4 0.4 0.4>
    overrideWorldGravity {z=-980}
   }
   n4 = Anim Dynamics - Bone: dyn_wobble_mass_jnt(ComponentPose=@n3, Alpha=@n2, ExternalForce="0,0,0") {
    angularDampingOverride=0.9999
    bGravityOverrideInSimSpace=true
    bOverrideAngularDamping=true
    bOverrideLinearDamping=true
    gravityScale=0.5
    linearDampingOverride=0.9999
    alphaBoolBlend {blendOption=Linear}
    boundBone {boneName=dyn_wobble_mass_jnt}
    componentAppliedLinearAccClamp=<8000 8000 8000>
    componentLinearAccScale=<0.15 0.15 0.15>
    componentLinearVelScale=<1 1 1>
    physicsBodyDefinitions [{
      sphereCollisionRadius=0
      boundBone {boneName=dyn_wobble_mass_jnt}
      boxExtents=<1 1 1>
      constraintSetup {angularLimitsMax=<0 0 0>, angularLimitsMin=<0 0 0>, angularTarget=<0 0 0>, linearAxesMax=<0 0 1.1>, linearAxesMin=<0 0 -1.1>}
      localJointOffset=<0 0 0>
     }]
    relativeSpaceBone {boneName=pack}
   }
   n5 = "Transform (Modify) Bone - Bone: dyn_base_rotate_jnt"(ComponentPose=@n4, Alpha=1, Translation="0,0,0", Rotation="0,0,10", Scale="1,1,1") {rotationMode=BMM_Additive, alphaBoolBlend {blendOption=Linear}, boneToModify {boneName=dyn_base_rotate_jnt}, rotation {roll=10}}
   n6 = "Transform (Modify) Bone - Bone: dyn_long_chain_1_jnt"(ComponentPose=@n5, Alpha=1, Translation="0,0,0", Rotation="0,0,5", Scale="1,1,1") {rotationMode=BMM_Additive, alphaBoolBlend {blendOption=Linear}, boneToModify {boneName=dyn_long_chain_1_jnt}, rotation {roll=5}}
   n7 = "Transform (Modify) Bone - Bone: dyn_short_chain_1_jnt"(ComponentPose=@n6, Alpha=1, Translation="0,0,0", Rotation="0,0,5", Scale="1,1,1") {rotationMode=BMM_Additive, alphaBoolBlend {blendOption=Linear}, boneToModify {boneName=dyn_short_chain_1_jnt}, rotation {roll=5}}
   Output Pose(Result=Component To Local(ComponentPose=@n7)) {layerGroup=None}
  ]}
 EventGraph {nodes [
   TryGetPawnOwner()
   Event BlueprintUpdateAnimation()
  ]}
]
variables [isRigidBodyDisabled {type=boolean}]
